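# Physics asset for the dynamic cape of the M_MED_Cavern_Armored back accessory, giving its cape bones simulated collision and motion limits.
# A large capsule on backpack_root anchors a four-bone center chain of capsules.
# Left and right five-bone chains both branch from dyn_cape03 rather than from the root.
# The center chain swings fairly freely, while the side chains are tightly limited and all joints allow little twist.
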PhysicsAsset /BRCosmetics/Accessories/FORT_Capes/M_MED_Cavern_Armored/Meshes/M_MED_Cavern_Armored_Cape_Dyn
bodies {
 backpack_root [{shapeType=Capsule, length=56.125, radius=37.4062, center=<-0.001 37.2328 -55.1791>, rotation=<rot 0 0 5.0186>}]
 dyn_cape01 [{shapeType=Capsule, length=15.97, radius=4.5076, center=<4.0879 -2.3 0>}]
 dyn_cape01_l [{shapeType=Capsule, length=7, radius=10, center=<10 -6.8 -10>}]
 dyn_cape01_r [{shapeType=Capsule, length=7, radius=10, center=<10 -6.8 10>}]
 dyn_cape02 [{shapeType=Capsule, length=15.97, radius=4.5076, center=<4.0879 -2.3 0>}]
 dyn_cape02_l [{shapeType=Capsule, length=10, radius=10, center=<4.0879 -6.8 -10>}]
 dyn_cape02_r [{shapeType=Capsule, length=10, radius=10, center=<4.0879 -6.8 10>}]
 dyn_cape03 [{shapeType=Capsule, length=15.97, radius=4.5076, center=<4.0879 -2.3 0>}]
 dyn_cape03_l [{shapeType=Capsule, length=13.5012, radius=10, center=<4.0879 -6.8 -10>}]
 dyn_cape03_r [{shapeType=Capsule, length=13.5012, radius=10, center=<4.0879 -6.8 10>}]
 dyn_cape04 [{shapeType=Capsule, length=14.47, radius=6.0076, center=<4.0879 -2.3 0>}]
 dyn_cape04_l [{shapeType=Capsule, length=20, radius=10, center=<4.0879 -6.8 -10>}]
 dyn_cape04_r [{shapeType=Capsule, length=20, radius=10, center=<4.0879 -6.8 10>}]
 dyn_cape05_l [{shapeType=Capsule, length=27, radius=10, center=<4.0879 -6.8 -10>}]
 dyn_cape05_r [{shapeType=Capsule, length=27, radius=10, center=<4.0879 -6.8 10>}]
}
constraints [
 {bone1Name=dyn_cape02, bone2Name=dyn_cape01, swing1LimitDegrees=45, swing2LimitDegrees=45, twistLimitDegrees=5}
 {bone1Name=dyn_cape03, bone2Name=dyn_cape02, swing1LimitDegrees=45, swing2LimitDegrees=45, twistLimitDegrees=5}
 {bone1Name=dyn_cape04, bone2Name=dyn_cape03, swing1LimitDegrees=45, swing2LimitDegrees=25, twistLimitDegrees=2}
 {bone1Name=dyn_cape02_l, bone2Name=dyn_cape01_l, swing1LimitDegrees=8, swing2LimitDegrees=8, twistLimitDegrees=2}
 {bone1Name=dyn_cape04_l, bone2Name=dyn_cape03_l, swing1LimitDegrees=10, swing2LimitDegrees=10, twistLimitDegrees=2}
 {bone1Name=dyn_cape02_r, bone2Name=dyn_cape01_r, swing1LimitDegrees=8, swing2LimitDegrees=8, twistLimitDegrees=2}
 {bone1Name=dyn_cape04_r, bone2Name=dyn_cape03_r, swing1LimitDegrees=10, swing2LimitDegrees=10, twistLimitDegrees=2}
 {bone1Name=dyn_cape01, bone2Name=backpack_root, swing1LimitDegrees=45, swing2LimitDegrees=45, twistLimitDegrees=5}
 {bone1Name=dyn_cape03_l, bone2Name=dyn_cape02_l, swing1LimitDegrees=10, swing2LimitDegrees=10, twistLimitDegrees=2}
 {bone1Name=dyn_cape05_l, bone2Name=dyn_cape04_l, swing1LimitDegrees=10, swing2LimitDegrees=10, twistLimitDegrees=2}
 {bone1Name=dyn_cape03_r, bone2Name=dyn_cape02_r, swing1LimitDegrees=10, swing2LimitDegrees=10, twistLimitDegrees=2}
 {bone1Name=dyn_cape05_r, bone2Name=dyn_cape04_r, swing1LimitDegrees=10, swing2LimitDegrees=10, twistLimitDegrees=2}
 {bone1Name=dyn_cape01_l, bone2Name=dyn_cape03, swing1LimitDegrees=8, swing2LimitDegrees=8, twistLimitDegrees=2}
 {bone1Name=dyn_cape01_r, bone2Name=dyn_cape03, swing1LimitDegrees=8, swing2LimitDegrees=8, twistLimitDegrees=2}
]
