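# Physics asset for the dynamic "Sky" variant of the M_MED_Cavern_Armored cape backpack cosmetic, defining its capsule bodies and joint limits.
# Two large capsules on backpack_root anchor a four-bone center chain that swings fairly freely until its tighter last joint.
# Left and right five-bone chains branch from the third center bone with very tight swing and twist limits, keeping the sides stiff, with capsules lengthening toward the ends.
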PhysicsAsset /BRCosmetics/Accessories/FORT_Capes/M_MED_Cavern_Armored/Meshes/M_MED_Cavern_Armored_Cape_Dyn_Sky
bodies {
 backpack_root [
  {shapeType=Capsule, length=42.125, radius=35.1562, center=<0.0008 15.4479 -62.4429>, rotation=<rot -0.0002 0 20.0229>}
  {shapeType=Capsule, length=42.125, radius=35.1562, center=<0.0002 -20.4135 -94.4158>, rotation=<rot 90 0 0>}
 ]
 dyn_cape01 [{shapeType=Capsule, length=15.97, radius=4.5076, center=<4.0879 -2.3 0>}]
 dyn_cape01_l [{shapeType=Capsule, length=8.5012, radius=10, center=<10 -6.8 -10>}]
 dyn_cape01_r [{shapeType=Capsule, length=8.5012, radius=10, center=<10 -6.8 10>}]
 dyn_cape02 [{shapeType=Capsule, length=15.97, radius=4.5076, center=<4.0879 -2.3 0>}]
 dyn_cape02_l [{shapeType=Capsule, length=11.7512, radius=10, center=<4.0879 -6.8 -10>}]
 dyn_cape02_r [{shapeType=Capsule, length=11.7512, radius=10, center=<4.0879 -6.8 10>}]
 dyn_cape03 [{shapeType=Capsule, length=15.97, radius=4.5076, center=<4.0879 -2.3 0>}]
 dyn_cape03_l [{shapeType=Capsule, length=18.7512, radius=10, center=<4.0879 -6.8 -10>}]
 dyn_cape03_r [{shapeType=Capsule, length=18.7512, radius=10, center=<4.0879 -6.8 10>}]
 dyn_cape04 [{shapeType=Capsule, length=13.72, radius=6.7576, center=<30 -2.3 0>}]
 dyn_cape04_l [{shapeType=Capsule, length=20.7512, radius=10, center=<4.0879 -6.8 -10>}]
 dyn_cape04_r [{shapeType=Capsule, length=20.7512, radius=10, center=<4.0879 -6.8 10>}]
 dyn_cape05_l [{shapeType=Capsule, length=28.2512, radius=10, center=<4.0879 -6.8 -10>}]
 dyn_cape05_r [{shapeType=Capsule, length=28.2512, radius=10, center=<4.0879 -6.8 10>}]
}
constraints [
 {bone1Name=dyn_cape02, bone2Name=dyn_cape01, swing1LimitDegrees=45, swing2LimitDegrees=45, twistLimitDegrees=5}
 {bone1Name=dyn_cape03, bone2Name=dyn_cape02, swing1LimitDegrees=45, swing2LimitDegrees=45, twistLimitDegrees=5}
 {bone1Name=dyn_cape04, bone2Name=dyn_cape03, swing1LimitDegrees=15, swing2LimitDegrees=10, twistLimitDegrees=2}
 {bone1Name=dyn_cape02_l, bone2Name=dyn_cape01_l, swing1LimitDegrees=5, swing2LimitDegrees=6, twistLimitDegrees=2}
 {bone1Name=dyn_cape04_l, bone2Name=dyn_cape03_l, swing1LimitDegrees=5, swing2LimitDegrees=6, twistLimitDegrees=2}
 {bone1Name=dyn_cape02_r, bone2Name=dyn_cape01_r, swing1LimitDegrees=5, swing2LimitDegrees=6, twistLimitDegrees=2}
 {bone1Name=dyn_cape04_r, bone2Name=dyn_cape03_r, swing1LimitDegrees=5, swing2LimitDegrees=6, twistLimitDegrees=2}
 {bone1Name=dyn_cape01, bone2Name=backpack_root, swing1LimitDegrees=45, swing2LimitDegrees=45, twistLimitDegrees=5}
 {bone1Name=dyn_cape03_l, bone2Name=dyn_cape02_l, swing1LimitDegrees=5, swing2LimitDegrees=6, twistLimitDegrees=2}
 {bone1Name=dyn_cape05_l, bone2Name=dyn_cape04_l, swing1LimitDegrees=5, swing2LimitDegrees=6, twistLimitDegrees=2}
 {bone1Name=dyn_cape03_r, bone2Name=dyn_cape02_r, swing1LimitDegrees=5, swing2LimitDegrees=6, twistLimitDegrees=2}
 {bone1Name=dyn_cape05_r, bone2Name=dyn_cape04_r, swing1LimitDegrees=5, swing2LimitDegrees=6, twistLimitDegrees=2}
 {bone1Name=dyn_cape01_l, bone2Name=dyn_cape03, swing1LimitDegrees=5, swing2LimitDegrees=6, twistLimitDegrees=2}
 {bone1Name=dyn_cape01_r, bone2Name=dyn_cape03, swing1LimitDegrees=5, swing2LimitDegrees=6, twistLimitDegrees=2}
]
